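# Sound cue for a laser beam countdown-off event, classed as a trap sound and governed by the Heist_LaserLines concurrency set.
# A mixer starts randomized countdown-off and countdown-on variants immediately, then adds random laser bursts and on-sweeteners after a fixed 1.9-second delay.
# Modulators sit on all layers except countdown-off, which instead has its own attenuation node.
# Cue-level attenuation overrides add HRTF spatialization, occlusion, listener focus and distance filtering.
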SoundCue /LaserGameplay/Sounds/Laser_Beam_Countdown_Off
root=Mixer_0
nodes [
 {
  id=Attenuation_0
  children [Random_0]
  properties {
   attenuationSettings=/Game/Sounds/Apollo/Biomes/Skyfire/Energy_Field_Loop_Close_Att
   bOverrideAttenuation=true
   attenuationOverrides {
    bAttenuateWithLPF=true
    bEnableFocusInterpolation=true
    bEnableListenerFocus=true
    bEnableOcclusion=true
    falloffDistance=1700
    focusAttackInterpSpeed=0.2
    focusReleaseInterpSpeed=0.2
    hPFFrequencyAtMax=400
    lPFFrequencyAtMax=3000
    lPFRadiusMax=1500
    lPFRadiusMin=200
    nonFocusVolumeAttenuation=0.6
    occlusionInterpolationTime=0.2
    occlusionLowPassFilterFrequency=5000
    occlusionVolumeAttenuation=0.8
    reverbDistanceMax=1500
    reverbDistanceMin=100
    spatializationAlgorithm=SPATIALIZATION_HRTF
    attenuationShapeExtents {x=100}
   }
  }
 }
 {id=Delay_0, children [Modulator_3], properties {delayMax=1.9, delayMin=1.9}}
 {id=Delay_1, children [Random_3], properties {delayMax=1.9, delayMin=1.9}}
 {id=Mixer_0, children [Attenuation_0, Delay_0, Modulator_1, Modulator_0], properties {inputVolume [1.5,1,1,0.6]}}
 {id=Modulator_0, children [Delay_1]}
 {id=Modulator_1, children [Random_2], properties {pitchMax=1, pitchMin=1}}
 {id=Modulator_3, children [Random_1]}
 {id=Random_0, children [WavePlayer_0,WavePlayer_2], properties {bShouldExcludeFromBranchCulling=true, weights [1,1]}}
 {
  id=Random_1
  children [WavePlayer_1,WavePlayer_3,WavePlayer_4,WavePlayer_6]
  properties {bShouldExcludeFromBranchCulling=true, weights [1,1,1,1]}
 }
 {id=Random_2, children [WavePlayer_5,WavePlayer_7], properties {bShouldExcludeFromBranchCulling=true, weights [1,1]}}
 {
  id=Random_3
  children [WavePlayer_8, WavePlayer_9, WavePlayer_10, WavePlayer_11, WavePlayer_12]
  properties {bShouldExcludeFromBranchCulling=true, weights [1,1,1,1,1]}
 }
 {id=WavePlayer_0, properties {soundWaveAssetPtr=/LaserGameplay/Sounds/wav/Laser_Lines_Countdown_Off_01}}
 {id=WavePlayer_1, properties {soundWaveAssetPtr=/LaserGameplay/Sounds/wav/Heist_Laser_Burst_04}}
 {id=WavePlayer_10, properties {soundWaveAssetPtr=/LaserGameplay/Sounds/wav/LaserLines_On_Sweetener_02}}
 {id=WavePlayer_11, properties {soundWaveAssetPtr=/LaserGameplay/Sounds/wav/LaserLines_On_Sweetener_03}}
 {id=WavePlayer_12, properties {soundWaveAssetPtr=/LaserGameplay/Sounds/wav/LaserLines_On_Sweetener_04}}
 {id=WavePlayer_2, properties {soundWaveAssetPtr=/LaserGameplay/Sounds/wav/Laser_Lines_Countdown_Off_02}}
 {id=WavePlayer_3, properties {soundWaveAssetPtr=/LaserGameplay/Sounds/wav/Heist_Laser_Burst_01}}
 {id=WavePlayer_4, properties {soundWaveAssetPtr=/LaserGameplay/Sounds/wav/Heist_Laser_Burst_02}}
 {id=WavePlayer_5, properties {soundWaveAssetPtr=/LaserGameplay/Sounds/wav/Laser_Lines_Countdown_On_01}}
 {id=WavePlayer_6, properties {soundWaveAssetPtr=/LaserGameplay/Sounds/wav/Heist_Laser_Burst_03}}
 {id=WavePlayer_7, properties {soundWaveAssetPtr=/LaserGameplay/Sounds/wav/Laser_Lines_Countdown_On_02}}
 {id=WavePlayer_8, properties {soundWaveAssetPtr=/LaserGameplay/Sounds/wav/LaserLines_On_Sweetener_05}}
 {id=WavePlayer_9, properties {soundWaveAssetPtr=/LaserGameplay/Sounds/wav/LaserLines_On_Sweetener_01}}
]
properties {
 attenuationSettings=&Energy_Field_Loop_Close_Att
 bOverrideAttenuation=true
 maxDistance=2100
 pitchMultiplier=0.93
 priority=0
 soundClassObject=&Fort_Traps
 volumeMultiplier=0.6
 attenuationOverrides {
  bAttenuateWithLPF=true
  bEnableFocusInterpolation=true
  bEnableListenerFocus=true
  bEnableOcclusion=true
  falloffDistance=2000
  focusAttackInterpSpeed=0.2
  focusReleaseInterpSpeed=0.2
  hPFFrequencyAtMax=400
  lPFFrequencyAtMax=3000
  lPFRadiusMax=1200
  lPFRadiusMin=200
  nonFocusVolumeAttenuation=0.8
  nonSpatializedRadiusEnd=20
  occlusionInterpolationTime=0.2
  occlusionLowPassFilterFrequency=5000
  occlusionTraceChannel=ECC_GameTraceChannel11
  occlusionVolumeAttenuation=0.6
  reverbDistanceMax=1000
  reverbDistanceMin=100
  reverbWetLevelMin=0.2
  spatializationAlgorithm=SPATIALIZATION_HRTF
  stereoSpread=250
  attenuationShapeExtents {x=100}
 }
 concurrencyOverrides {
  bLimitToOwner=true
  bVolumeScaleCanRelease=true
  maxCount=1
  resolutionRule=StopQuietest
  voiceStealReleaseTime=0.1
  volumeScale=0.6
  volumeScaleReleaseTime=0
 }
 concurrencySet [/LaserGameplay/Sounds/Settings/Heist_LaserLines]
}
