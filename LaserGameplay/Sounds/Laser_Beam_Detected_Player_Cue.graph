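# Laser_Beam_Detected_Player_Cue is a gameplay-event SoundCue that, going by its name, plays when a laser beam detects a player.
# A single wave player feeds the Laser_Lines_Detected wave straight to the output, using Heist laser-line attenuation and concurrency assets with raised priority.
# Its concurrency overrides specify one instance per owner, blocking new plays while one is active, plus a short retrigger time.
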SoundCue /LaserGameplay/Sounds/Laser_Beam_Detected_Player_Cue
root=WavePlayer_0
nodes [{id=WavePlayer_0, properties {soundWaveAssetPtr=/LaserGameplay/Sounds/wav/Laser_Lines_Detected}}]
properties {
 attenuationSettings=/LaserGameplay/Sounds/Settings/Heist_LaserLines_Detected
 maxDistance=5700
 priority=2
 soundClassObject=&Fort_Events_Gameplay
 volumeMultiplier=0.7
 concurrencyOverrides {bLimitToOwner=true, maxCount=1, resolutionRule=PreventNew, retriggerTime=1}
 concurrencySet [/LaserGameplay/Sounds/Settings/Heist_LaserLines]
}
