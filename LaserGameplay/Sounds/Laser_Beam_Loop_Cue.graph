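# Looping laser beam Sound Cue in the traps sound class, layering two looped waves into one continuous hum.
# Each wave passes through slight random volume modulation and a short fixed delay into a low-gain mixer, with the Laser_Lines layer pitched down.
# Inline attenuation overrides the referenced settings with HRTF spatialization, occlusion, distance filtering and listener focus.
# Concurrency overrides allow one voice per owner, stopping the farthest when exceeded.
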SoundCue /LaserGameplay/Sounds/Laser_Beam_Loop_Cue
root=Mixer_1
nodes [
 {id=Delay_0, children [Modulator_1], properties {delayMax=0.1, delayMin=0.1}}
 {id=Delay_1, children [Modulator_2], properties {delayMax=0.1, delayMin=0.1}}
 {id=Mixer_1, children [Delay_0,Delay_1], properties {inputVolume [0.2,0.2]}}
 {id=Modulator_1, children [WavePlayer_5], properties {pitchMax=1, pitchMin=1, volumeMax=1, volumeMin=0.9}}
 {id=Modulator_2, children [WavePlayer_1], properties {pitchMax=0.9, pitchMin=0.9, volumeMax=1, volumeMin=0.9}}
 {id=WavePlayer_1, properties {bLooping=true, soundWaveAssetPtr=/LaserGameplay/Sounds/wav/Laser_Lines_Loop}}
 {id=WavePlayer_5, properties {bLooping=true, soundWaveAssetPtr=/LaserGameplay/Sounds/wav/Heist_Laser_Beam_Loop_Close_01}}
]
properties {
 attenuationSettings=/Game/Sounds/Apollo/Biomes/Skyfire/Energy_Field_Loop_Close_Att
 bOverrideAttenuation=true
 bOverrideConcurrency=true
 maxDistance=2100
 pitchMultiplier=0.93
 priority=0
 soundClassObject=&Fort_Traps
 volumeMultiplier=0.7
 attenuationOverrides {
  bAttenuateWithLPF=true
  bEnableFocusInterpolation=true
  bEnableListenerFocus=true
  bEnableOcclusion=true
  falloffDistance=2000
  focusAttackInterpSpeed=0.2
  focusReleaseInterpSpeed=0.2
  hPFFrequencyAtMax=600
  lPFFrequencyAtMax=5000
  lPFRadiusMax=1200
  lPFRadiusMin=200
  nonFocusVolumeAttenuation=0.8
  nonSpatializedRadiusEnd=20
  occlusionInterpolationTime=0.2
  occlusionLowPassFilterFrequency=5000
  occlusionTraceChannel=ECC_GameTraceChannel11
  occlusionVolumeAttenuation=0.6
  reverbDistanceMax=1000
  reverbDistanceMin=100
  reverbWetLevelMin=0.2
  spatializationAlgorithm=SPATIALIZATION_HRTF
  stereoSpread=250
  attenuationShapeExtents {x=100}
 }
 concurrencyOverrides {
  bLimitToOwner=true
  bVolumeScaleCanRelease=true
  maxCount=1
  resolutionRule=StopFarthestThenPreventNew
  voiceStealReleaseTime=0.1
  volumeScale=0.6
  volumeScaleReleaseTime=0
 }
}
