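# Sound cue for the laser-line disabled sound in LaserGameplay, routed to the traps sound class.
# A fixed short delay feeds a modulator at default settings, which plays the LaserLines_Disabled wave.
# Inline attenuation overrides the assigned energy-field attenuation asset, adding HRTF spatialization, occlusion, listener focus and distance filtering.
# The shared Heist_LaserLines concurrency set governs simultaneous playback, as the cue's own concurrency override is not enabled.
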SoundCue /LaserGameplay/Sounds/Laser_Disabled
root=Delay_0
nodes [
 {id=Delay_0, children [Modulator_3], properties {delayMax=0.2, delayMin=0.2}}
 {id=Modulator_3, children [WavePlayer_0]}
 {id=WavePlayer_0, properties {soundWaveAssetPtr=/LaserGameplay/Sounds/wav/LaserLines_Disabled}}
]
properties {
 attenuationSettings=/Game/Sounds/Apollo/Biomes/Skyfire/Energy_Field_Loop_Close_Att
 bOverrideAttenuation=true
 maxDistance=2300
 soundClassObject=&Fort_Traps
 volumeMultiplier=1.7
 attenuationOverrides {
  bAttenuateWithLPF=true
  bEnableFocusInterpolation=true
  bEnableListenerFocus=true
  bEnableOcclusion=true
  falloffDistance=1500
  focusAttackInterpSpeed=0.5
  focusReleaseInterpSpeed=0.5
  hPFFrequencyAtMax=300
  lPFFrequencyAtMax=2000
  lPFRadiusMax=1500
  lPFRadiusMin=600
  nonFocusVolumeAttenuation=0.7
  nonSpatializedRadiusEnd=200
  occlusionInterpolationTime=0.2
  occlusionLowPassFilterFrequency=2000
  occlusionTraceChannel=ECC_GameTraceChannel6
  occlusionVolumeAttenuation=0.7
  reverbDistanceMax=600
  reverbDistanceMin=100
  reverbWetLevelMin=0.2
  spatializationAlgorithm=SPATIALIZATION_HRTF
  stereoSpread=250
  attenuationShapeExtents {x=800}
 }
 concurrencyOverrides {bVolumeScaleCanRelease=true, maxCount=1, resolutionRule=PreventNew, voiceStealReleaseTime=0.1, volumeScale=0.8, volumeScaleReleaseTime=0}
 concurrencySet [/LaserGameplay/Sounds/Settings/Heist_LaserLines]
}
