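# Sound cue for a laser outlet light-on sound in the traps sound class, with an empty node graph, so it plays no audio as saved.
# Its attenuation override supersedes the assigned Skyfire energy-field attenuation asset, adding HRTF spatialization, distance low-pass, occlusion and reverb send.
# A single-voice, per-owner concurrency block is filled in, but its override flag is left off.
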SoundCue /LaserGameplay/Sounds/Laser_Outlet_LightOn
root=null
nodes []
properties {
 attenuationSettings=/Game/Sounds/Apollo/Biomes/Skyfire/Energy_Field_Loop_Close_Att
 bOverrideAttenuation=true
 maxDistance=1300
 priority=0
 soundClassObject=&Fort_Traps
 volumeMultiplier=0.7
 attenuationOverrides {
  bAttenuateWithLPF=true
  bEnableOcclusion=true
  falloffDistance=1200
  lPFFrequencyAtMax=3000
  lPFFrequencyAtMin=10000
  lPFRadiusMax=100
  lPFRadiusMin=50
  nonSpatializedRadiusEnd=20
  occlusionLowPassFilterFrequency=2000
  occlusionTraceChannel=ECC_GameTraceChannel6
  occlusionVolumeAttenuation=0.6
  reverbDistanceMax=1000
  reverbDistanceMin=100
  reverbWetLevelMin=0.2
  spatializationAlgorithm=SPATIALIZATION_HRTF
  attenuationShapeExtents {x=100}
 }
 concurrencyOverrides {bLimitToOwner=true, bVolumeScaleCanRelease=true, maxCount=1, resolutionRule=StopQuietest, voiceStealReleaseTime=0.1}
}
